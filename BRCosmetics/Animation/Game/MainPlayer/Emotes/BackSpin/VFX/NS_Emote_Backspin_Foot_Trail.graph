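# Niagara ribbon trail for the Backspin emote, named as a foot trail.
# One emitter spawns particles at a steady rate on bones of the user-supplied skeletal mesh, without inheriting velocity.
# A curve over the emitter's looped age sets each particle's short lifetime, shrinking as the loop progresses.
# Ribbon width and RGB fade over life at low alpha; the dynamic material parameter module is disabled.
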NiagaraSystem /BRCosmetics/Animation/Game/MainPlayer/Emotes/BackSpin/VFX/NS_Emote_Backspin_Foot_Trail
dependencies {usedDataInterfaces [&NiagaraDataInterfaceCurve], usedDynamicInputs [], usedModules [&SystemState]}
emitters {NE_Empty001 {
  emitterUpdateScript [
   &EmitterState()
   &SpawnRate(SpawnRate=20, Use Spawn Probability=false)
  ]
  particleSpawnScript [
   &SkeletalMeshLocation(Bone Sampling Mode=::NewEnumerator4, Inherit Velocity=false, MaskSpawnGroup=false, Skeletal Mesh=$User.SkeletalMesh)
   &InitializeRibbon(
    Lifetime=&FloatFromCurve(
     FloatCurve=DataInterface() {propertyValues {Curve {keys [{arriveTangent=0.0737, interpMode=RCIM_Cubic, leaveTangent=0.0737, tangentMode=RCTM_User, time=0, value=1}, <at 0.995 0.5073>]}}}
     CurveIndex=$Emitter.LoopedAge
     Scale Curve=0.25
    )
    Module.Write Color=true
    Module.Write Lifetime=true
    Module.Write Mass=true
    Module.Write Material Random=true
    Module.Write Position=false
    Module.Write RibbonTwist=false
    Module.Write RibbonWidth=true
    Ribbon Width=30
   )
   &DynamicMaterialParameters(
    Index 0 Param 1=&Lerp_Float(A=2, B=2, Alpha=&MakeFloatFromInt(Int=$Particles.SpawnGroup))
    Module.Param0WriteEnabled=true
    Module.Param1WriteEnabled=false
    Module.Param2WriteEnabled=false
    Module.Param3WriteEnabled=false
   ) {enabled=false}
  ]
  particleUpdateScript [
   &ParticleState()
   &ScaleRibbonWidth(Ribbon Width Scale=&FloatFromCurve(FloatCurve=DataInterface() {propertyValues {Curve {keys [<at 0 1 RCIM_Cubic>,<at 1 0>]}}}))
   &ScaleColor(
    Scale Alpha=0.15
    Scale RGB=&VectorFromFloat(Value=&FloatFromCurve(FloatCurve=DataInterface() {propertyValues {Curve {keys [<at 0 1>,<at 1 0>]}}}))
    ScaleA=true
    ScaleRGB=true
   )
  ]
  rendererClasses [&NiagaraRibbonRendererProperties]
  renderers [{rendererClass=&NiagaraRibbonRendererProperties}]
 }}
summary {
 systemName=NS_Emote_Backspin_Foot_Trail
 emitters [{emitterName=NE_Empty001, rendererClasses [&NiagaraRibbonRendererProperties]}]
 userVariables [
  User.RibbonIDOffset {type=NiagaraInt32}
  User.SkeletalMesh {
   type=NiagaraDataInterfaceSkeletalMesh
   defaultValue {
    struct=NiagaraExt_VariableValue_DataInterface
    value {dataInterface=&self:NiagaraDataInterfaceSkeletalMesh_0}
   }
  }
 ]
}
